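# Additive, unlit, two-sided glow material for the CrewCut small bug zapper, drawing a soft tinted halo.
# A powered radial sphere mask, multiplied by a near-camera fade and softened by depth fade, drives opacity, while Color feeds emissive through FX adjustment functions.
# World position offset combines Sprite billboarding, scaled by green vertex color, with a camera offset that varies with view angle against a vector parameter.
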Material /Hermes/Props/CrewCut/CrewCut_SmallZapper_A/Materials/M_AirGlow_BugZapper
graph {
 properties {
  bAutomaticallySetUsageInEditor=false
  bUseMaterialAttributes=true
  blendMode=BLEND_Additive
  shadingModel=MSM_Unlit
  twoSided=true
  usedShadingModels=Unlit
 }
 nodes [
  ActorPositionWS()
  CameraVectorWS()
  n1 = CameraPositionWS()
  n2 = WorldPosition()
  n3 = DotProduct(A=VectorParameter() {defaultValue=<rgba 0 0 -1 0>}.RGB, B=Normalize(VectorInput=Subtract(A=@n1, B=@n2.XYZ)))
  LinearInterpolate(Alpha=Power(Base=Saturate(None=@n3), Exp=ScalarParameter() {defaultValue=1})) {constB=0}
  n4 = VectorParameter() {parameterName=Color, defaultValue=<rgba 0 0.1034 0.467 1>}
  n5 = TextureCoordinate()
  n6 = Constant() {r=0.5}
  n7 = ScalarParameter() {defaultValue=0.5, parameterName=Radius}
  n8 = ScalarParameter() {parameterName=Hardness}
  n9 = Multiply(
   A=Power(Base=SphereMask(A=@n5, B=@n6, Radius=@n7, Hardness=@n8), Exp=ScalarParameter() {defaultValue=2, parameterName=Mask Hardness})
   B=MaterialFunctionCall(Distance=ScalarParameter() {defaultValue=-0.5867, parameterName=Near Camera Fade}) {materialFunction=&Near_Camera_Fade}.Result
  )
  n10 = ScalarParameter() {defaultValue=10, parameterName=Fade distane}
  n11 = NamedRerouteUsage()
  n12 = NamedRerouteUsage()
  MaterialFunctionCall!(
   SFX
   Material In=SetMaterialAttributes!(
    MaterialAttributes
    Emissive Color=MaterialFunctionCall(Emissive=@n4.RGB) {materialFunction=&MF_FX_EmissiveAdjustments}.Result
    Opacity=DepthFade(Opacity=@n9, FadeDistance=@n10)
    World Position Offset=Add(A=@n11, B=@n12)
   ) {attributeSetTypes [#1, #2, #3]}
   SFX Textures RGB
  ) {materialFunction=/Game/Global/Base4Layers_2022/MaterialFunctions/MF_EmissiveControls_2}
  n13 = ScalarParameter() {defaultValue=75, parameterName=XY Scale}
  n14 = VertexColor()
  NamedRerouteDeclaration(None=MaterialFunctionCall!(Center Location, 0-1 UVs, XY scale=LinearInterpolate(B=@n13, Alpha=@n14.G), OverridePivot, "Normals (Optional)") {materialFunction=&Sprite}.World position offset) {name=Sprite Mesh WPO, nodeColor=<rgba 0 1 0.1562 1>}
  n15 = ScalarParameter() {defaultValue=25, parameterName=Camera Offset}
  n16 = MaterialFunctionCall!(Use World Position Offsets) {materialFunction=/Engine/Functions/Engine_MaterialFunctions02/Utility/CameraVectorWithWPOOptions}
  n17 = VectorParameter() {parameterName=Param_1, defaultValue=<rgba 0 0 1 0>}
  n18 = Power(Base=Saturate(None=DotProduct(A=@n16.Result, B=@n17.RGB)), Exp=ScalarParameter() {defaultValue=2, parameterName=Param_1})
  NamedRerouteDeclaration(None=MaterialFunctionCall(Offset Amount=LinearInterpolate(B=@n15, Alpha=@n18) {constA=-10}) {materialFunction=/Engine/Functions/Engine_MaterialFunctions02/WorldPositionOffset/CameraOffset}.Camera Offset) {name=CameraOffset, nodeColor=<rgba 1 0 0.1641 1>}
  Transform() {transformSourceType=TRANSFORMSOURCE_World, transformType=TRANSFORM_Local}
  WorldPosition()
 ]
}
