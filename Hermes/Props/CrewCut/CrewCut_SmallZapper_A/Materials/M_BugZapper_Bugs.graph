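# Masked, unlit, two-sided material for the insects around the CrewCut small bug zapper, drawn from a SubUV wasp swarm texture.
# A time-of-day scalar stepped against vertex alpha gates the sprite's alpha, tying the bugs' visibility to night.
# World position offset sums camera-facing sprite expansion, sized by vertex green, with panning curl-noise drift and rotation around the object pivot.
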Material /Hermes/Props/CrewCut/CrewCut_SmallZapper_A/Materials/M_BugZapper_Bugs
graph {
 properties {bAutomaticallySetUsageInEditor=false, blendMode=BLEND_Masked, shadingModel=MSM_Unlit, twoSided=true, usedShadingModels=Unlit}
 nodes [
  LinearInterpolate(B=MaterialFunctionCall() {materialFunction=&DitherTemporalAA}.Result)
  n1 = Constant()
  n2 = Constant() {r=1}
  Multiply(
   A=TextureSample(UVs=MaterialFunctionCall!(U_Images, V_Images, SubUV Texture RGB, Time Speed, Manual Time Control, Time, UVs) {materialFunction=/Game/Effects/Fort_Effects/Materials/_Functions/MF_SubUV}.UVs) {texture=/Game/Effects/Fort_Effects/Effects/Environment/Bugs/BugBiome/Materials/T_AI_Wildlife_InsectSwarm_SingleWasp_Baked_V4_1}.A
   B=Step(
    Y=VertexColor().A
    X=MaterialFunctionCall(Value_Morning=@n1, Value_Day=@n1, Value_Evening=@n1, Value_Night=@n2) {materialFunction=&Scalar_Time_of_Day}.Result
   )
  )
  n3 = MaterialFunctionCall!(
   Center Location=WorldPosition().XYZ
   0-1 UVs=TextureCoordinate()
   XY scale=LinearInterpolate(
    A=ScalarParameter() {defaultValue=2.5, parameterName=Min Scale}
    B=ScalarParameter() {defaultValue=8, parameterName=Max Scale}
    Alpha=VertexColor().G
   )
   OverridePivot
   "Normals (Optional)"
  ) {materialFunction=&Sprite}
  n4 = NamedRerouteUsage()
  NamedRerouteDeclaration(None=Add(A=Add(A=@n3.World position offset, B=@n4), B=NamedRerouteUsage())) {name=WPO, nodeColor=<rgba 1 0 0.3281 1>}
  n5 = Multiply(A=Time(), B=ScalarParameter() {defaultValue=0.5, parameterName=Curl Time Scale})
  n6 = VertexColor()
  n7 = Add(
   A=Multiply(A=WorldPosition().XYZ, B=ScalarParameter() {defaultValue=0.002, parameterName=Noise Scale})
   B=Multiply(A=VectorParameter() {parameterName=Curl Pan, defaultValue=<rgba 0 0 0.1 0>}.RGB, B=Add(A=@n5, B=@n6.G))
  )
  NamedRerouteDeclaration(None=Multiply(
    A=TextureSample(UVs=@n7) {samplerType=SAMPLERTYPE_LinearColor, texture=/Game/Effects/FogVolume/VT_CurlNoise_LOW}.RGB
    B=ScalarParameter() {defaultValue=10, parameterName=Curl Scale}
   )) {name=Noise, nodeColor=<rgba 0.5703 1 0 1>}
  n8 = Time()
  n9 = ScalarParameter() {defaultValue=0.25, parameterName=Rotation Scale}
  NamedRerouteDeclaration(None=RotateAboutAxis(
    NormalizedRotationAxis=VectorParameter() {parameterName=Rotation Axis, defaultValue=<rgba 0 0 1 0>}.RGB
    RotationAngle=Multiply(A=@n8, B=@n9)
    PivotPoint=ObjectPositionWS()
    Position=WorldPosition().XYZ
   )) {name=Rotate About Axis, nodeColor=<rgba 0.5469 0 1 1>}
  NamedRerouteUsage()
 ]
}
